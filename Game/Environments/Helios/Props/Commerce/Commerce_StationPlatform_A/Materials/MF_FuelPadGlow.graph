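# Material function for the fuel pad glow on the Commerce station platform, replacing emissive and optionally darkening base color on input material attributes.
# Combines the incoming emissive, used as a mask, with faked light glow from two local-space sphere masks shaped by surface normals.
# A fueling color tints the result, modulated by a slow sine pulse and timed on/off fades driven through custom primitive data timestamps and state.
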MaterialFunction /Game/Environments/Helios/Props/Commerce/Commerce_StationPlatform_A/Materials/MF_FuelPadGlow
graph {nodes [
  n1 = GetMaterialAttributes(None=NamedRerouteUsage()) {attributeGetTypes [#1, #2, #3]}
  n2 = MaterialFunctionCall(Float3=@n1.EmissiveColor) {materialFunction=&BreakOutFloat3Components}
  n3 = Multiply(A=Max(A=Max(A=@n2.R, B=@n2.R), B=@n2.B)) {constB=5}
  n4 = NamedRerouteDeclaration(None=@n3) {name=EmissiveMask, nodeColor=<rgba 0 1 0.3672 1>}
  n5 = LocalPosition() {localOrigin=InstancePreSkinning}
  n6 = ShaderStageSwitch(PixelShader=VertexInterpolator(VS=@n5.XYZ).PS, VertexShader=@n5.XYZ)
  n7 = VectorParameter() {parameterName=LightPos_01, defaultValue=<rgba 0 128 0 1>}
  n8 = ScalarParameter() {defaultValue=200, parameterName=LightGlowRadius}
  n9 = ScalarParameter() {defaultValue=0.5, parameterName=LightFalloff}
  n10 = VectorParameter() {parameterName=LightPos_02, defaultValue=<rgba 0 -128 0 1>}
  n11 = Power(Base=Max(A=SphereMask(A=@n6, B=@n7.RGB, Radius=@n8, Hardness=@n9), B=SphereMask(A=@n6, B=@n10.RGB, Radius=@n8, Hardness=@n9))) {constExponent=1.5}
  n12 = Transform(None=@n1.Normal) {transformSourceType=TRANSFORMSOURCE_World, transformType=TRANSFORM_Local}
  n13 = Subtract(A=@n6, B=@n10.RGB)
  n14 = Saturate(None=DotProduct(A=@n12, B=Normalize(VectorInput=@n13)))
  n15 = Subtract(A=@n6, B=@n7.RGB)
  n16 = Saturate(None=DotProduct(A=@n12, B=Normalize(VectorInput=@n15)))
  n17 = Multiply(A=Multiply(A=Saturate(None=@n11), B=Max(A=@n14, B=@n16))) {constB=0.25}
  n18 = Time()
  n19 = ScalarParameter() {defaultValue=1, parameterName=EmissiveOn-OffDuration}
  n20 = Divide(A=Subtract(A=@n18, B=ScalarParameter() {bUseCustomPrimitiveData=true, parameterName=EmissiveOffTimer, primitiveDataIndex=1}), B=@n19)
  n21 = Subtract(A=@n18, B=ScalarParameter() {bUseCustomPrimitiveData=true, parameterName=EmissiveOnTimer})
  n22 = LinearInterpolate(
   A=OneMinus(None=Saturate(None=@n20))
   B=Saturate(None=Divide(A=@n21, B=@n19))
   Alpha=ScalarParameter() {bUseCustomPrimitiveData=true, parameterName=EmissiveCurrentState, primitiveDataIndex=2}
  )
  n23 = Multiply(A=Time()) {constB=0.5}
  n24 = LinearInterpolate(Alpha=ConstantBiasScale(None=Sine(None=@n23))) {constA=0.1}
  n25 = Multiply(
   A=Multiply(
    A=VectorParameter() {parameterName=FuelChargeEmissive, defaultValue=<rgba 0 0.4362 1 1>}.RGB
    B=ScalarParameter() {defaultValue=10, parameterName=FuelingGlowStrength}
   )
   B=Multiply(A=Max(A=@n4, B=@n17), B=Multiply(A=@n22, B=@n24))
  )
  n26 = NamedRerouteUsage()
  n27 = Multiply(A=@n26, B=ScalarParameter() {defaultValue=1, parameterName=BaseColorDarken})
  n28 = NamedRerouteUsage()
  FunctionOutput(None=SetMaterialAttributes!(MaterialAttributes=NamedRerouteUsage(), EmissiveColor=Power(Base=@n25), BaseColor=LinearInterpolate(A=@n26, B=@n27, Alpha=@n28)) {attributeSetTypes [#1, #3]})
  Multiply(A=@n14, B=@n16)
  NamedRerouteDeclaration(None=GetMaterialAttributes(None=FunctionInput() {inputName=Material, inputType=FunctionInput_MaterialAttributes}).MaterialAttributes) {name=Input_Material, nodeColor=<rgba 1 0.4922 0 1>}
  NamedRerouteDeclaration(None=@n1.BaseColor) {name=Input_BaseColor, nodeColor=<rgba 0.4531 0 1 1>}
  Saturate(None=@n3)
  ScalarParameter() {defaultValue=1, parameterName=EmissiveOn}
 ]}
